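; BP_Apollo_BlowingDandelion is an environment effect actor that shows blowing dandelion motes within a spawn area.
; Its construction script scales the root by the spawn-area variable, then creates a dynamic material instance on the motes mesh.
; Exposed size, drift speed, horizontal drift range and overall speed variables feed that material's scalar parameters, so motion is presumably material-driven.
; BeginPlay, overlap and Tick events are empty; all setup happens at construction.
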(event EventBeginPlay)

(event Collision|EventActorBeginOverlap
  (params (OtherActor :type Actor%20Object%20Reference)))

(event EventTick
  (params (DeltaSeconds :type Float%20%28single-precision%29)))

(fn ConstructionScript
  (Transformation|SetWorldScale3D (Variables|Default|GetDefaultSceneRoot) (Variables|Default|GetScaleDandelionSpawnArea))
  (bind _returnvalue (Rendering|Material|CreateDynamicMaterialInstance (Variables|Default|GetSM_Apollo_Motes_01) 0 (Rendering|Material|GetMaterial (Variables|Default|GetSM_Apollo_Motes_01) 0) "None" :!class /Script/Engine.PrimitiveComponent))
  (Rendering|Material|SetScalarParameterValue _returnvalue "MaxSize" (Variables|Default|GetDandelionSize) :!class /Script/Engine.MaterialInstanceDynamic)
  (Rendering|Material|SetScalarParameterValue _returnvalue "DriftSpeed" (Variables|Default|GetDandelionDriftSpeed) :!class /Script/Engine.MaterialInstanceDynamic)
  (Rendering|Material|SetScalarParameterValue _returnvalue "MaxHorizontalDrift" (Variables|Default|GetMaxHorizontalDrift) :!class /Script/Engine.MaterialInstanceDynamic)
  (Rendering|Material|SetScalarParameterValue _returnvalue "MinHorizontalDrift" (Variables|Default|GetMinHorizontalDrift) :!class /Script/Engine.MaterialInstanceDynamic)
  (Rendering|Material|SetScalarParameterValue _returnvalue "OverallSpeed" (Variables|Default|GetOverallSpeed) :!class /Script/Engine.MaterialInstanceDynamic))
